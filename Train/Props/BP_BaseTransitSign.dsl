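; Base transit sign that tracks the time until the next train reaches its station on a linked train rail.
; Once playlist data is ready, it binds rail arrival-interval and departure events, filtered by station tag, to update the time.
; The server also seeds a replicated arrival time, refreshed on train registration.
; An unresolved rail starts a retry timer, but the retry reschedules only at 10 or more attempts, so it effectively runs once.
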(event Custom|OnTrainRegistered
  (params (Train :type Fort%20Train%20Manager%20Object%20Reference))
  (if (Utilities|IsValidSoftObjectReference (Variables|Default|GetTrainRail))
    (Utilities|FlowControl|Delay 1.0)
    (bind _output_get (Variables|FortTrainTransitSign|SetTimeToArrival (Class|FortTrainRail|GetTimefromNextTrainArrival (Utilities|ResolveSoftReference (Variables|Default|GetTrainRail)) (Variables|Default|GetTrainStationIdentifier))))
    (Variables|Default|SetServerTimeToArrival _output_get)))

(event EventBeginPlay
  (bind _trainrail (Variables|Default|GetTrainRail))
  (bind _output (Utilities|ResolveSoftReference _trainrail))
  (Class|FortAsyncActionCurrentPlaylistReady|PlaylistDataReadyAsync
    (:OnReady
      (Variables|Default|SetTrainRailResolveAttempts 0)
      (if (Utilities|IsValidSoftObjectReference (Variables|Default|GetTrainRail))
        (EventDispatchers|BindEventtoOnTrainStationArrivalIntervalReachedEvent _output (EventDispatchers|CreateEvent :!event Custom|OnStationArrivalIntervalReached))
        (EventDispatchers|BindEventtoOnTrainStationDepartureEvent _output (EventDispatchers|CreateEvent :!event Custom|OnTrainStationDeparture))
        (Networking|SwitchHasAuthority
          (:Authority
            (EventDispatchers|BindEventtoOnTrainRegisteredEvent _output (EventDispatchers|CreateEvent :!event Custom|OnTrainRegistered))
            (bind _output_get (Variables|FortTrainTransitSign|SetTimeToArrival (Class|FortTrainRail|GetTimefromNextTrainArrival _output (Variables|Default|GetTrainStationIdentifier))))
            (Variables|Default|SetServerTimeToArrival _output_get)))
        (else
          (bind _returnvalue (Utilities|Time|SetTimerbyEvent (delegate Custom|RetryResolveTrainRail) (Variables|Default|GetResolveTrainRailRetryTimer)))
          (Variables|Default|SetResolveTrainRailTimer _returnvalue)
          (Math|Integer|IncrementInt (Variables|Default|GetTrainRailResolveAttempts)))))))

(event Custom|RetryResolveTrainRail
  (bind _trainrail (Variables|Default|GetTrainRail))
  (bind _output (Utilities|ResolveSoftReference _trainrail))
  (if (Utilities|IsValidSoftObjectReference (Variables|Default|GetTrainRail))
    (EventDispatchers|BindEventtoOnTrainStationArrivalIntervalReachedEvent _output (EventDispatchers|CreateEvent :!event Custom|OnStationArrivalIntervalReached))
    (EventDispatchers|BindEventtoOnTrainStationDepartureEvent _output (EventDispatchers|CreateEvent :!event Custom|OnTrainStationDeparture))
    (Networking|SwitchHasAuthority
      (:Authority
        (EventDispatchers|BindEventtoOnTrainRegisteredEvent _output (EventDispatchers|CreateEvent :!event Custom|OnTrainRegistered))
        (bind _output_get (Variables|FortTrainTransitSign|SetTimeToArrival (Class|FortTrainRail|GetTimefromNextTrainArrival _output (Variables|Default|GetTrainStationIdentifier))))
        (Variables|Default|SetServerTimeToArrival _output_get)))
    (elif (< (Variables|Default|GetTrainRailResolveAttempts) 10)
      _
      (else
        (bind _returnvalue (Utilities|Time|SetTimerbyEvent (EventDispatchers|CreateEvent :!event Custom|RetryResolveTrainRail) (Variables|Default|GetResolveTrainRailRetryTimer)))
        (Variables|Default|SetResolveTrainRailTimer _returnvalue)
        (Math|Integer|IncrementInt (Variables|Default|GetTrainRailResolveAttempts))))))

(event EventEndPlay
  (params (EndPlayReason :type EEndPlayReason%20Enum :default "Destroyed"))
  (Utilities|Time|ClearandInvalidateTimerbyHandle (Variables|Default|GetResolveTrainRailTimer)))

(event Custom|OnStationArrivalIntervalReached
  (params (Train :type Fort%20Train%20Manager%20Object%20Reference) (StationIdentifier :type Gameplay%20Tag%20Structure) (TimeUntilArrival :type Float%20%28single-precision%29) (Interval :type Float%20%28single-precision%29))
  (if (== StationIdentifier (Variables|Default|GetTrainStationIdentifier))
    (Networking|SwitchHasAuthority
      (:Authority
        _)
      (:Remote
        (Variables|FortTrainTransitSign|SetTimeToArrival TimeUntilArrival)))
    (CallFunction|OnArrivalTimeIntervalReachedforStation TimeUntilArrival)))

(event Custom|OnTrainStationDeparture
  (params (Train :type Fort%20Train%20Manager%20Object%20Reference) (StationIdentifier :type Gameplay%20Tag%20Structure) (TimeUntilNextArrival :type Float%20%28double-precision%29))
  (if (== StationIdentifier (Variables|Default|GetTrainStationIdentifier))
    (if (Utilities|IsValidSoftObjectReference (Variables|Default|GetTrainRail))
      (Variables|FortTrainTransitSign|SetTimeToArrival TimeUntilNextArrival))))

(event Custom|OnArrivalTimeIntervalReachedForStation
  (params (TImeUntilArrival :type Float%20%28double-precision%29)))

(event EventOnDeathPlayEffects
  (params (Damage :type Float%20%28single-precision%29) (DamageTags :type Gameplay%20Tag%20Container%20Structure%20%28by%20ref%29) (Momentum :type Vector :default "0, 0, 0") (HitInfo :type Hit%20Result%20Structure%20%28by%20ref%29) (InstigatedBy :type Fort%20Pawn%20Object%20Reference) (DamageCauser :type Actor%20Object%20Reference) (EffectContext :type Gameplay%20Effect%20Context%20Handle%20Structure))
  (Rendering|SetActorHiddenInGame true))

(event OnLinkedActorDestroyed(FortLinkToActor)
  (params (DamageCauser :type Actor%20Object%20Reference))
  (Actor|SetLifeSpan self 0.25))

(fn OnRep_ServerTimeToArrival
  (Variables|FortTrainTransitSign|SetTimeToArrival (Variables|Default|GetServerTimeToArrival)))

(fn ConstructionScript)
